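; Day 3 Blueprint variant of the BrickSoupSpoon block prop, which reacts to a beacon trigger.
; The TriggerBeacon custom event prints a debug string, switches the golden-ray Niagara effect to a new RayState, and briefly fades out the idle audio.
; BeginPlay, overlap and tick add no logic, and the construction script only calls the parent.
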(event Custom|TriggerBeacon
  (Development|PrintString "TriggerBeacon")
  (Niagara|SetNiagaraVariable(Int32) (Variables|Default|GetNS_BrickSoup_GoldenRay) "RayState" 4)
  (Audio|Components|Audio|FadeOut (Variables|Default|GetIdleComponent) 0.25 0.0 "Linear" :!class /Script/Engine.AudioComponent))

(event EventBeginPlay
  (|CallParentFunction) ; This node is disabled and will not be called.\nDrag off pins to build functionality.
)

(event Collision|EventActorBeginOverlap
  (params (OtherActor :type Actor%20Object%20Reference)))

(event EventTick
  (params (DeltaSeconds :type Float%20%28single-precision%29)))

(fn ConstructionScript
  (|CallParentFunction))
